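# DADBRO_Music_Intro_Cue is a sound cue for the intro music of the Dadbro event, reusing a Storm King orchestral loop track.
# A fixed two-second delay feeds a single wave player for the StormKing_Into_the_Storm loop.
# It plays at reduced volume on the quest music sound class, with non-spatialized box attenuation overrides and a very large max distance.
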SoundCue /Game/Sounds/Apollo/Events/Dadbro/Music/DADBRO_Music_Intro_Cue
root=Delay_0
nodes [
 {id=Delay_0, children [WavePlayer_1], properties {delayMax=2, delayMin=2}}
 {
  id=WavePlayer_1
  properties {soundWaveAssetPtr=/Game/Sounds/Fort_Music/Orchestral_Score/Orch_Storm_King/Final_MIxes/StormKing_Into_the_Storm_01_Loop}
 }
]
properties {
 maxDistance=2097152
 soundClassObject=/Game/Sounds/FortSoundClassesAndModes/Fort_Music_Quest_PSM
 volumeMultiplier=0.15
 attenuationOverrides {attenuationShape=Box, bSpatialize=false, falloffDistance=500, attenuationShapeExtents=<2200 2200 1500>}
}
