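# DadBro_Combat_Music_02_Cue is a sound cue for combat music in the DadBro event.
# A single looping wave player plays the Storm King "Reinforcements Arrived" orchestral loop at reduced volume.
# It routes through the Fort_Music_Quest_PSM sound class, with spatialization disabled in its attenuation overrides.
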SoundCue /Game/Sounds/Apollo/Events/Dadbro/Music/DadBro_Combat_Music_02_Cue
root=WavePlayer_0
nodes [{
  id=WavePlayer_0
  properties {
   bLooping=true
   soundWaveAssetPtr=/Game/Sounds/Fort_Music/Orchestral_Score/Orch_Storm_King/Final_MIxes/StormKing_Reinforcments_Arrived_01_Loop
  }
 }]
properties {
 maxDistance=2097152
 soundClassObject=/Game/Sounds/FortSoundClassesAndModes/Fort_Music_Quest_PSM
 volumeMultiplier=0.5
 attenuationOverrides {bSpatialize=false}
}
