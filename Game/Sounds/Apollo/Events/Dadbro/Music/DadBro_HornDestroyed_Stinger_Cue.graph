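# DadBro_HornDestroyed_Stinger_Cue is a Sound Cue for a music stinger in the Dadbro event, named for the destruction of a horn.
# A Delay node holds playback for a fixed half second, then a single Wave Player plays a Storm King horn-destroyed orchestral music wave.
# The cue routes to the boss fight sound class at slightly reduced volume, with spatialization off and a very large max distance.
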SoundCue /Game/Sounds/Apollo/Events/Dadbro/Music/DadBro_HornDestroyed_Stinger_Cue
root=Delay_0
nodes [
 {id=Delay_0, children [WavePlayer_1], properties {delayMax=0.5, delayMin=0.5}}
 {
  id=WavePlayer_1
  properties {soundWaveAssetPtr=/Game/Sounds/Fort_Music/Orchestral_Score/Orch_Storm_King/Final_MIxes/StormKing_Horn_Destroyed_Music_02}
 }
]
properties {
 maxDistance=2097152
 soundClassObject=&Fort_BossFight
 volumeMultiplier=0.8
 attenuationOverrides {bSpatialize=false}
}
