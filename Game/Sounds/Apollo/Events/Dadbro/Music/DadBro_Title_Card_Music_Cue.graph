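# Sound cue for the DadBro event title card music, reusing the Storm King reveal orchestral track.
# A single wave player plays StormKing_Reveal_01 as the root with no further processing nodes.
# It routes through the Fort_BossFight sound class at reduced volume.
# Its attenuation override turns spatialization off, and a very large max distance is set.
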SoundCue /Game/Sounds/Apollo/Events/Dadbro/Music/DadBro_Title_Card_Music_Cue
root=WavePlayer_0
nodes [{id=WavePlayer_0, properties {soundWaveAssetPtr=/Game/Sounds/Fort_Music/Orchestral_Score/Orch_Storm_King/Final_MIxes/StormKing_Reveal_01}}]
properties {
 maxDistance=2097152
 soundClassObject=&Fort_BossFight
 volumeMultiplier=0.6
 attenuationOverrides {bSpatialize=false}
}
