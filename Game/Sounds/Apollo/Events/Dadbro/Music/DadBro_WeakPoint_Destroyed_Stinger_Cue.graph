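# Sound cue for the music stinger that, per its name, plays when a DadBro boss weak point is destroyed.
# A short fixed delay feeds a random node that picks, with equal weight, between two Storm King weak point reveal stinger waves.
# It plays at reduced volume in the boss fight sound class, non-spatialized with a very large max distance.
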SoundCue /Game/Sounds/Apollo/Events/Dadbro/Music/DadBro_WeakPoint_Destroyed_Stinger_Cue
root=Delay_0
nodes [
 {id=Delay_0, children [Random_0], properties {delayMax=0.75, delayMin=0.75}}
 {id=Random_0, children [WavePlayer_3,WavePlayer_0], properties {weights [1,1]}}
 {
  id=WavePlayer_0
  properties {soundWaveAssetPtr=/Game/Sounds/Fort_Music/Orchestral_Score/Orch_Storm_King/Final_MIxes/StormKing_Weak_Point_Reveal_Stinger_01}
 }
 {
  id=WavePlayer_3
  properties {soundWaveAssetPtr=/Game/Sounds/Fort_Music/Orchestral_Score/Orch_Storm_King/Final_MIxes/StormKing_Weak_Point_Reveal_Stinger_02}
 }
]
properties {
 maxDistance=2097152
 soundClassObject=&Fort_BossFight
 volumeMultiplier=0.6
 attenuationOverrides {bSpatialize=false}
}
